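# Niagara effect for the Weightless emote: colorful mesh balls that float around the player and burst into smaller ones.
# Main_Ball fires one timed radial burst, while Support_Big_Balls spawns floating, wobbling balls at eight staggered offsets.
# Big balls write deaths to a system array, driving Support_Small_Balls bursts at those positions.
# Meshes shrink when the camera nears the owner; duplicate support emitters are disabled.
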NiagaraSystem /BRCosmetics/Animation/Game/MainPlayer/Emotes/Weightless/VFX/NS_Weightless_01
dependencies {
 usedDataInterfaces [
  &NiagaraDataInterfaceArrayColor
  &NiagaraDataInterfaceCurve
  &NiagaraDataInterfaceCamera
  &NiagaraDataInterfaceVectorCurve
  &NiagaraDataInterfaceArrayPosition
 ]
 usedDynamicInputs [&Multiply_Vector2D_ByFloat, &RandomRangeVector2D]
 usedModules [
  &self:SystemSpawnScript.SystemScriptSource.SystemScriptGraph.NiagaraNodeAssignment_0.SetVariables_#1
  &SystemState
  &self:ReadDeath
  &self:SystemSpawnScript.SystemScriptSource.SystemScriptGraph.NiagaraNodeAssignment_2.SetVariables_#2
  &self:SystemSpawnScript.SystemScriptSource.SystemScriptGraph.NiagaraNodeAssignment_1.SetVariables_#3
 ]
}
emitters {
 Main_Ball {
  emitterUpdateScript [
   &EmitterState(Life Cycle Mode=::NewEnumerator1, Loop Behavior=::NewEnumerator1, Loop Delay=0.4, Loop Duration Mode=::NewEnumerator1, UseLoopDelay=false)
   &SpawnBurst_Instantaneous(Spawn Count=50, Spawn Time=5.18, Use Spawn Probability=false, UseLoopCountLimit=false)
  ]
  particleSpawnScript [
   &InitializeParticle(
    Color=&FNTimeOfDayColor(
     Night Color=&SelectLinearColorFromArray(
      Array Sampling Mode=::NewEnumerator0
      Direct Array Index=?
      Color Selection Array=DataInterface() {propertyValues {ColorData [
         <rgba 0.6038 0.0908 0 1>
         <rgba 0.0307 0.2874 0.0369 1>
         <rgba 0.0723 0.0823 0.9301 1>
         <rgba 0.3372 0.0742 0.4969 1>
         <rgba 0.8963 0.4287 0.0097 1>
         <rgba 0 1 1 1>
         <rgba 0.9823 0.0467 0.4621 1>
        ]}}
      Random Seed=0
     )
     Morning Color=&SelectLinearColorFromArray(
      Array Sampling Mode=::NewEnumerator0
      Direct Array Index=?
      Color Selection Array=DataInterface() {propertyValues {ColorData [
         <rgba 0.6038 0.0908 0 1>
         <rgba 0.0307 0.2874 0.0369 1>
         <rgba 0.0723 0.0823 0.9301 1>
         <rgba 0.3372 0.0742 0.4969 1>
         <rgba 0.8963 0.4287 0.0097 1>
         <rgba 0 1 1 1>
         <rgba 0.9823 0.0467 0.4621 1>
        ]}}
      Random Seed=0
     )
     Day Color=&SelectLinearColorFromArray(
      Array Sampling Mode=::NewEnumerator0
      Direct Array Index=?
      Color Selection Array=DataInterface() {propertyValues {ColorData [
         <rgba 0.6038 0.0908 0 1>
         <rgba 0.0307 0.2874 0.0369 1>
         <rgba 0.0723 0.0823 0.9301 1>
         <rgba 0.3372 0.0742 0.4969 1>
         <rgba 0.8963 0.4287 0.0097 1>
         <rgba 0 1 1 1>
         <rgba 0.9823 0.0467 0.4621 1>
        ]}}
      Random Seed=0
     )
     Evening Color=&SelectLinearColorFromArray(
      Array Sampling Mode=::NewEnumerator0
      Direct Array Index=?
      Color Selection Array=DataInterface() {propertyValues {ColorData [
         <rgba 0.6038 0.0908 0 1>
         <rgba 0.0307 0.2874 0.0369 1>
         <rgba 0.0723 0.0823 0.9301 1>
         <rgba 0.3372 0.0742 0.4969 1>
         <rgba 0.8963 0.4287 0.0097 1>
         <rgba 0.1565 1 1 1>
         <rgba 0.9823 0.0467 0.4621 1>
        ]}}
      Random Seed=0
     )
    )
    Lifetime Max=5.2
    Lifetime Min=4.8
    Lifetime Mode=::NewEnumerator1
    Mesh Scale Mode=::NewEnumerator3
    Mesh Uniform Scale=&Multiply_Float(A=0.1, B=&RandomRangeFloat(Minimum=1, Maximum=1))
    Sprite Rotation Mode=::NewEnumerator2
    UsePositionOffset=false
    Write Material Random=true
   )
   &ShapeLocation(MaskSpawnGroup=false, Offset=<0 70 109>, Offset Mode=::NewEnumerator0, Override Seed=false, Sphere Radius=2, UseOwnerScale=true)
   &AddVelocity(
    Cone Angle=180
    Cone Axis=<0 1 0>
    Module.UseConeFalloff=false
    Module.UseVelDistribution=false
    Override Seed=false
    Velocity Mode=::NewEnumerator2
    Velocity Speed=&RandomRangeFloat(Minimum=163, Maximum=168)
   )
   &ApplyInitialForces()
  ]
  particleUpdateScript [
   &ParticleState()
   &ScaleMeshSize(Scale Factor=&VectorFromFloat(Value=&FloatFromCurve(FloatCurve=DataInterface() {propertyValues {Curve {keys [{interpMode=RCIM_Cubic, tangentMode=RCTM_Break, time=0.9, value=1}, {interpMode=RCIM_Cubic, tangentMode=RCTM_Break, time=1, value=0}]}}})))
   &ScaleMeshSize(Scale Factor=&VectorFromFloat(Value=&Lerp_Float(
      A=0
      B=1
      Alpha=&RemapRange(
       Input Value To Remap=/Niagara/DynamicInputs/Camera/DistanceToCamera(Use Owner Position:NiagaraBool=true, Camera Query:NiagaraDataInterfaceCamera=DataInterface())
       Input Min=50
       Input Max=240
       Output Min=0
       Output Max=1
      )
     ))) {module=ScaleMeshSize001}
   &Drag(Drag=&RandomRangeFloat(Minimum=0.5, Maximum=0.7), Use Linear Drag=true, Use Rotational Drag=false)
   &SolveForcesAndVelocity(Clamp Velocity=false, Limit Acceleration=false)
  ]
  rendererClasses [&NiagaraMeshRendererProperties]
  renderers [{rendererClass=&NiagaraMeshRendererProperties, rendererIndex=0}]
 }
 Main_Orb {rendererClasses [], renderers []}
 Support_Big_Balls {
  emitterUpdateScript [
   &EmitterState(Life Cycle Mode=::NewEnumerator1, Loop Duration=10.5, UseLoopDelay=false)
   &SpawnBurst_Instantaneous(Spawn Time=0.5, Use Spawn Probability=false, UseLoopCountLimit=false)
   &SpawnBurst_Instantaneous(Spawn Group=1, Spawn Time=1.5, Use Spawn Probability=false, UseLoopCountLimit=false) {module=SpawnBurst_Instantaneous001}
   &SpawnBurst_Instantaneous(Spawn Group=2, Spawn Time=2.5, Use Spawn Probability=false, UseLoopCountLimit=false) {module=SpawnBurst_Instantaneous002}
   &SpawnBurst_Instantaneous(Spawn Group=3, Spawn Time=3.5, Use Spawn Probability=false, UseLoopCountLimit=false) {module=SpawnBurst_Instantaneous003}
   &SpawnBurst_Instantaneous(Spawn Group=4, Spawn Time=4.5, Use Spawn Probability=false, UseLoopCountLimit=false) {module=SpawnBurst_Instantaneous004}
   &SpawnBurst_Instantaneous(Spawn Group=5, Spawn Time=5.5, Use Spawn Probability=false, UseLoopCountLimit=false) {module=SpawnBurst_Instantaneous005}
   &SpawnBurst_Instantaneous(Spawn Group=6, Spawn Time=6.5, Use Spawn Probability=false, UseLoopCountLimit=false) {module=SpawnBurst_Instantaneous006}
   &SpawnBurst_Instantaneous(Spawn Group=7, Spawn Time=7.5, Use Spawn Probability=false, UseLoopCountLimit=false) {module=SpawnBurst_Instantaneous007}
  ]
  particleSpawnScript [
   &PartitionParticles(Number Of Partitions=8)
   &InitializeParticle(
    Color=&SelectLinearColorFromArray(
     Array Sampling Mode=::NewEnumerator0
     Direct Array Index=?
     Color Selection Array=DataInterface() {propertyValues {ColorData [
        <rgba 0.901 0.901 0.901 1>
        <rgba 0.0307 0.2874 0.0369 1>
        <rgba 0.0723 0.0823 0.9301 1>
        <rgba 0.3372 0.0742 0.4969 1>
        <rgba 0.8963 0.4287 0.0097 1>
       ]}}
     Random Seed=0
    )
    Lifetime=&MaskFloat(
     Mask Source=0
     Mask Parameter=?
     Mask=0
     Masked Float=5.4
     Passthrough Float=&MaskFloat(
      Mask Source=0
      Mask Parameter=?
      Mask=1
      Masked Float=4.8
      Passthrough Float=&MaskFloat(
       Mask Source=0
       Mask Parameter=?
       Mask=2
       Masked Float=4.2
       Passthrough Float=&MaskFloat(
        Mask Source=0
        Mask Parameter=?
        Mask=3
        Masked Float=3.6
        Passthrough Float=&MaskFloat(
         Mask Source=0
         Mask Parameter=?
         Mask=4
         Masked Float=3
         Passthrough Float=&MaskFloat(
          Mask Source=0
          Mask Parameter=?
          Mask=5
          Masked Float=2.4
          Passthrough Float=&MaskFloat(
           Mask Source=0
           Mask Parameter=?
           Mask=6
           Masked Float=1.8
           Passthrough Float=&MaskFloat(Mask Source=0, Mask Parameter=?, Mask=7, Masked Float=1.2, Passthrough Float=1)
          )
         )
        )
       )
      )
     )
    )
    Mesh Scale=<0.26 0.26 0.26>
    Mesh Scale Mode=::NewEnumerator4
    UsePositionOffset=false
    Write Material Random=true
   )
   &AddVelocity(Velocity=&RandomRangeVector(Minimum=<-2 -2 15>, Maximum=<2 2 36>))
   &SystemLocation(MaskSpawnGroup=true, Offset=<110 0 10>)
   &SystemLocation(MaskSpawnGroup=true, Offset=<-110 8 0>, Spawn Group Mask=1) {module=SystemLocation001}
   &SystemLocation(MaskSpawnGroup=true, Offset=<70 115 15>, Spawn Group Mask=2) {module=SystemLocation002}
   &SystemLocation(MaskSpawnGroup=true, Offset=<40 -110 40>, Spawn Group Mask=3) {module=SystemLocation003}
   &SystemLocation(MaskSpawnGroup=true, Offset=<-50 -100 0>, Spawn Group Mask=4) {module=SystemLocation004}
   &SystemLocation(MaskSpawnGroup=true, Offset=<110 0 0>, Spawn Group Mask=5) {module=SystemLocation005}
   &SystemLocation(MaskSpawnGroup=true, Offset=<-110 8 0>, Spawn Group Mask=6) {module=SystemLocation006}
   &SystemLocation(MaskSpawnGroup=true, Offset=<70 110 15>, Spawn Group Mask=7) {module=SystemLocation007}
  ]
  particleUpdateScript [
   &ParticleState()
   &self:WriteDeath(Position Array:NiagaraDataInterfaceArrayPosition=$System.DeathArray) {module=WriteDeath}
   &ScaleMeshSize(Scale Factor=&VectorFromFloat(Value=&FloatFromCurve(FloatCurve=DataInterface() {propertyValues {Curve {keys [{interpMode=RCIM_Cubic, tangentMode=RCTM_Break, time=0, value=0}, {interpMode=RCIM_Cubic, tangentMode=RCTM_Break, time=0.3, value=1}]}}}))) {module=ScaleMeshSize001}
   &ScaleMeshSize(Scale Factor=&VectorFromCurve(VectorCurve=DataInterface() {propertyValues {
       XCurve {keys [{interpMode=RCIM_Cubic, tangentMode=RCTM_Break, time=0, value=0.6778}, {interpMode=RCIM_Cubic, tangentMode=RCTM_Break, time=0.24, value=1}]}
       YCurve {keys [{interpMode=RCIM_Cubic, tangentMode=RCTM_Break, time=0, value=0.6778}, {interpMode=RCIM_Cubic, tangentMode=RCTM_Break, time=0.24, value=1}]}
       ZCurve {keys [
         {interpMode=RCIM_Cubic, leaveTangent=-0.0045, tangentMode=RCTM_Break, time=0, value=0.6778}
         {arriveTangent=0.0244, interpMode=RCIM_Cubic, tangentMode=RCTM_Break, time=0.1011, value=1.1}
         {arriveTangent=0.0212, interpMode=RCIM_Cubic, leaveTangent=0.0212, tangentMode=RCTM_User, time=0.2215, value=0.72}
         {arriveTangent=-0.0273, interpMode=RCIM_Cubic, leaveTangent=-0.0273, tangentMode=RCTM_User, time=0.3236, value=1}
         {interpMode=RCIM_Cubic, tangentMode=RCTM_Break, time=1, value=1}
        ]}
      }}))
   &ScaleMeshSize(Scale Factor=&VectorFromFloat(Value=&Lerp_Float(
      A=0
      B=1
      Alpha=&RemapRange(
       Input Value To Remap=&DistanceToCamera(Use Owner Position:NiagaraBool=true, Camera Query:NiagaraDataInterfaceCamera=DataInterface())
       Input Min=50
       Input Max=240
       Output Min=0
       Output Max=1
      )
     ))) {module=ScaleMeshSize002}
   &GravityForce(Gravity=&RandomRangeVector(Minimum=<-10 -10 0>, Maximum=<10 10 5>, Evaluation Type=::NewEnumerator1))
   &SolveForcesAndVelocity(Clamp Velocity=false, Limit Acceleration=false)
  ]
  rendererClasses [&NiagaraMeshRendererProperties]
  renderers [{rendererClass=&NiagaraMeshRendererProperties, rendererIndex=0}]
 }
 Support_Big_Balls001 {
  bEnabled=false
  emitterUpdateScript [
   &EmitterState(Life Cycle Mode=::NewEnumerator1, Loop Duration=10.5, UseLoopDelay=false)
   &SpawnBurst_Instantaneous(Spawn Time=0.5, Use Spawn Probability=false, UseLoopCountLimit=false)
   &SpawnBurst_Instantaneous(Spawn Group=1, Spawn Time=1.5, Use Spawn Probability=false, UseLoopCountLimit=false) {module=SpawnBurst_Instantaneous001}
   &SpawnBurst_Instantaneous(Spawn Group=2, Spawn Time=2.5, Use Spawn Probability=false, UseLoopCountLimit=false) {module=SpawnBurst_Instantaneous002}
   &SpawnBurst_Instantaneous(Spawn Group=3, Spawn Time=3.5, Use Spawn Probability=false, UseLoopCountLimit=false) {module=SpawnBurst_Instantaneous003}
   &SpawnBurst_Instantaneous(Spawn Group=4, Spawn Time=4.5, Use Spawn Probability=false, UseLoopCountLimit=false) {module=SpawnBurst_Instantaneous004}
   &SpawnBurst_Instantaneous(Spawn Group=5, Spawn Time=5.5, Use Spawn Probability=false, UseLoopCountLimit=false) {module=SpawnBurst_Instantaneous005}
   &SpawnBurst_Instantaneous(Spawn Group=6, Spawn Time=6.5, Use Spawn Probability=false, UseLoopCountLimit=false) {module=SpawnBurst_Instantaneous006}
   &SpawnBurst_Instantaneous(Spawn Group=7, Spawn Time=7.5, Use Spawn Probability=false, UseLoopCountLimit=false) {module=SpawnBurst_Instantaneous007}
  ]
  particleSpawnScript [
   &PartitionParticles(Number Of Partitions=8)
   &InitializeParticle(
    Color=&SelectLinearColorFromArray(
     Array Sampling Mode=::NewEnumerator0
     Direct Array Index=?
     Color Selection Array=DataInterface() {propertyValues {ColorData [
        <rgba 0.901 0.901 0.901 1>
        <rgba 0.0307 0.2874 0.0369 1>
        <rgba 0.0723 0.0823 0.9301 1>
        <rgba 0.3372 0.0742 0.4969 1>
        <rgba 0.8963 0.4287 0.0097 1>
       ]}}
     Random Seed=0
    )
    Lifetime=&MaskFloat(
     Mask Source=0
     Mask Parameter=?
     Mask=0
     Masked Float=5.4
     Passthrough Float=&MaskFloat(
      Mask Source=0
      Mask Parameter=?
      Mask=1
      Masked Float=4.8
      Passthrough Float=&MaskFloat(
       Mask Source=0
       Mask Parameter=?
       Mask=2
       Masked Float=4.2
       Passthrough Float=&MaskFloat(
        Mask Source=0
        Mask Parameter=?
        Mask=3
        Masked Float=3.6
        Passthrough Float=&MaskFloat(
         Mask Source=0
         Mask Parameter=?
         Mask=4
         Masked Float=3
         Passthrough Float=&MaskFloat(
          Mask Source=0
          Mask Parameter=?
          Mask=5
          Masked Float=2.4
          Passthrough Float=&MaskFloat(
           Mask Source=0
           Mask Parameter=?
           Mask=6
           Masked Float=1.8
           Passthrough Float=&MaskFloat(Mask Source=0, Mask Parameter=?, Mask=7, Masked Float=1.2, Passthrough Float=1)
          )
         )
        )
       )
      )
     )
    )
    Mesh Scale=<0.26 0.26 0.26>
    Mesh Scale Mode=::NewEnumerator4
    UsePositionOffset=false
    Write Material Random=true
   )
   &AddVelocity(Velocity=&RandomRangeVector(Minimum=<-2 -2 15>, Maximum=<2 2 36>))
   &SystemLocation(MaskSpawnGroup=true, Offset=<110 0 10>)
   &SystemLocation(MaskSpawnGroup=true, Offset=<-110 8 0>, Spawn Group Mask=1) {module=SystemLocation001}
   &SystemLocation(MaskSpawnGroup=true, Offset=<70 115 15>, Spawn Group Mask=2) {module=SystemLocation002}
   &SystemLocation(MaskSpawnGroup=true, Offset=<40 -110 40>, Spawn Group Mask=3) {module=SystemLocation003}
   &SystemLocation(MaskSpawnGroup=true, Offset=<-50 -100 0>, Spawn Group Mask=4) {module=SystemLocation004}
   &SystemLocation(MaskSpawnGroup=true, Offset=<110 0 0>, Spawn Group Mask=5) {module=SystemLocation005}
   &SystemLocation(MaskSpawnGroup=true, Offset=<-110 8 0>, Spawn Group Mask=6) {module=SystemLocation006}
   &SystemLocation(MaskSpawnGroup=true, Offset=<70 110 15>, Spawn Group Mask=7) {module=SystemLocation007}
  ]
  particleUpdateScript [
   &ParticleState()
   &self:WriteDeath(Position Array:NiagaraDataInterfaceArrayPosition=$System.DeathArray) {module=WriteDeath}
   &ScaleMeshSize(Scale Factor=&VectorFromFloat(Value=&FloatFromCurve(FloatCurve=DataInterface() {propertyValues {Curve {keys [{interpMode=RCIM_Cubic, tangentMode=RCTM_Break, time=0, value=0}, {interpMode=RCIM_Cubic, tangentMode=RCTM_Break, time=0.3, value=1}]}}}))) {module=ScaleMeshSize001}
   &ScaleMeshSize(Scale Factor=&VectorFromCurve(VectorCurve=DataInterface() {propertyValues {
       XCurve {keys [{interpMode=RCIM_Cubic, tangentMode=RCTM_Break, time=0, value=0.6778}, {interpMode=RCIM_Cubic, tangentMode=RCTM_Break, time=0.24, value=1}]}
       YCurve {keys [{interpMode=RCIM_Cubic, tangentMode=RCTM_Break, time=0, value=0.6778}, {interpMode=RCIM_Cubic, tangentMode=RCTM_Break, time=0.24, value=1}]}
       ZCurve {keys [
         {interpMode=RCIM_Cubic, leaveTangent=-0.0045, tangentMode=RCTM_Break, time=0, value=0.6778}
         {arriveTangent=0.0244, interpMode=RCIM_Cubic, tangentMode=RCTM_Break, time=0.1011, value=1.1}
         {arriveTangent=0.0212, interpMode=RCIM_Cubic, leaveTangent=0.0212, tangentMode=RCTM_User, time=0.2215, value=0.72}
         {arriveTangent=-0.0273, interpMode=RCIM_Cubic, leaveTangent=-0.0273, tangentMode=RCTM_User, time=0.3236, value=1}
         {interpMode=RCIM_Cubic, tangentMode=RCTM_Break, time=1, value=1}
        ]}
      }}))
   &ScaleMeshSize(Scale Factor=&VectorFromFloat(Value=&Lerp_Float(
      A=0
      B=1
      Alpha=&RemapRange(
       Input Value To Remap=&DistanceToCamera(Use Owner Position:NiagaraBool=true, Camera Query:NiagaraDataInterfaceCamera=DataInterface())
       Input Min=50
       Input Max=240
       Output Min=0
       Output Max=1
      )
     ))) {module=ScaleMeshSize002}
   &GravityForce(Gravity=&RandomRangeVector(Minimum=<-10 -10 0>, Maximum=<10 10 5>, Evaluation Type=::NewEnumerator1))
   &SolveForcesAndVelocity(Clamp Velocity=false, Limit Acceleration=false)
  ]
  rendererClasses [&NiagaraMeshRendererProperties]
  renderers [{rendererClass=&NiagaraMeshRendererProperties, rendererIndex=0}]
 }
 Support_Small_Balls {
  emitterUpdateScript [
   &EmitterState(Life Cycle Mode=::NewEnumerator1, Loop Behavior=::NewEnumerator1, Loop Duration Mode=::NewEnumerator1, UseLoopDelay=false)
   &SpawnPerFrame(Spawn:NiagaraBool=$System.DeathSpawn, Spawn Count=30, Spawn Group=1, Use Spawn Probability=false)
  ]
  particleSpawnScript [
   &PartitionParticles(Number Of Partitions=8, Particles Per Partition=30, Partition Mode:ENiagara_PartitionMode_v2=::NewEnumerator1)
   &InitializeParticle(
    Color=&SelectLinearColorFromArray(
     Array Sampling Mode=::NewEnumerator0
     Direct Array Index=?
     Color Selection Array=DataInterface() {propertyValues {ColorData [
        <rgba 0.901 0.901 0.901 1>
        <rgba 0.0307 0.2874 0.0369 1>
        <rgba 0.0723 0.0823 0.9301 1>
        <rgba 0.3372 0.0742 0.4969 1>
        <rgba 0.8963 0.4287 0.0097 1>
       ]}}
     Random Seed=0
    )
    Lifetime=&MaskFloat(
     Mask Source=1
     Mask Parameter=$Particles.PartitionParticles.Partition
     Mask=0
     Masked Float=&RandomRangeFloat(Minimum=4.3, Maximum=4.7)
     Passthrough Float=&MaskFloat(
      Mask Source=1
      Mask Parameter=$Particles.PartitionParticles.Partition
      Mask=1
      Masked Float=&RandomRangeFloat(Minimum=3.7, Maximum=4.1)
      Passthrough Float=&MaskFloat(
       Mask Source=1
       Mask Parameter=$Particles.PartitionParticles.Partition
       Mask=2
       Masked Float=&RandomRangeFloat(Minimum=3.1, Maximum=3.5)
       Passthrough Float=&MaskFloat(
        Mask Source=1
        Mask Parameter=$Particles.PartitionParticles.Partition
        Mask=3
        Masked Float=&RandomRangeFloat(Minimum=2.5, Maximum=2.9)
        Passthrough Float=&MaskFloat(
         Mask Source=1
         Mask Parameter=$Particles.PartitionParticles.Partition
         Mask=4
         Masked Float=&RandomRangeFloat(Minimum=2.3, Maximum=2.7)
         Passthrough Float=&MaskFloat(
          Mask Source=1
          Mask Parameter=$Particles.PartitionParticles.Partition
          Mask=5
          Masked Float=&RandomRangeFloat(Minimum=2, Maximum=2.4)
          Passthrough Float=&MaskFloat(
           Mask Source=1
           Mask Parameter=$Particles.PartitionParticles.Partition
           Mask=6
           Masked Float=&RandomRangeFloat(Minimum=1.6, Maximum=2)
           Passthrough Float=&MaskFloat(
            Mask Source=1
            Mask Parameter=$Particles.PartitionParticles.Partition
            Mask=7
            Masked Float=&RandomRangeFloat(Minimum=1.1, Maximum=1.5)
            Passthrough Float=1
           )
          )
         )
        )
       )
      )
     )
    )
    Mesh Scale Mode=::NewEnumerator3
    Mesh Uniform Scale=&Multiply_Float(A=0.08, B=&RandomRangeFloat(Minimum=0.95, Maximum=1))
    Position=$System.DeathPosition
    Position Mode=::NewEnumerator1
    UsePositionOffset=false
    Write Material Random=true
   )
   &ShapeLocation(MaskSpawnGroup=false, Override Seed=false, Sphere Radius=1, UseOwnerScale=true)
   &AddVelocity(
    Module.Use Falloff Distance=false
    Origin Offset Coordinate Space=::Simulation
    Override Seed=false
    Velocity Mode=::NewEnumerator1
    Velocity Speed=&RandomRangeFloat(Minimum=110, Maximum=130)
   )
  ]
  particleUpdateScript [
   &ParticleState()
   &ScaleMeshSize(Scale Factor=&VectorFromFloat(Value=&FloatFromCurve(FloatCurve=DataInterface() {propertyValues {Curve {keys [{interpMode=RCIM_Cubic, tangentMode=RCTM_Break, time=0.9, value=1}, {interpMode=RCIM_Cubic, tangentMode=RCTM_Break, time=1, value=0}]}}})))
   &ScaleMeshSize(Scale Factor=&VectorFromFloat(Value=&Lerp_Float(
      A=0
      B=1
      Alpha=&RemapRange(
       Input Value To Remap=&DistanceToCamera(Use Owner Position:NiagaraBool=true, Camera Query:NiagaraDataInterfaceCamera=DataInterface())
       Input Min=50
       Input Max=240
       Output Min=0
       Output Max=1
      )
     ))) {module=ScaleMeshSize002}
   &GravityForce(Gravity=&RandomVector(Vector Scale=&FloatFromCurve(
      FloatCurve=DataInterface() {propertyValues {Curve {keys [
          {interpMode=RCIM_Cubic, tangentMode=RCTM_Break, time=0, value=1}
          {arriveTangent=-2.0604, interpMode=RCIM_Cubic, tangentMode=RCTM_Break, time=0.9, value=0}
         ]}}}
      Scale Curve=&RandomRangeFloat(Minimum=50, Maximum=70)
     )))
   &Drag(Drag=&RandomRangeFloat(Minimum=0.85, Maximum=1.45), Use Linear Drag=true, Use Rotational Drag=false)
   &SolveForcesAndVelocity(Clamp Velocity=false, Limit Acceleration=false)
  ]
  rendererClasses [&NiagaraMeshRendererProperties]
  renderers [{rendererClass=&NiagaraMeshRendererProperties, rendererIndex=0}]
 }
 Support_Small_Balls001 {
  bEnabled=false
  emitterUpdateScript [
   &EmitterState(Life Cycle Mode=::NewEnumerator1, Loop Behavior=::NewEnumerator1, Loop Duration Mode=::NewEnumerator1, UseLoopDelay=false)
   &SpawnPerFrame(Spawn:NiagaraBool=$System.DeathSpawn, Spawn Count=20, Use Spawn Probability=false)
  ]
  particleSpawnScript [
   &PartitionParticles(Number Of Partitions=8, Particles Per Partition=20, Partition Mode:ENiagara_PartitionMode_v2=::NewEnumerator1)
   &InitializeParticle(
    Color=&SelectLinearColorFromArray(
     Array Sampling Mode=::NewEnumerator0
     Direct Array Index=?
     Color Selection Array=DataInterface() {propertyValues {ColorData [
        <rgba 0.901 0.901 0.901 1>
        <rgba 0.0307 0.2874 0.0369 1>
        <rgba 0.0723 0.0823 0.9301 1>
        <rgba 0.3372 0.0742 0.4969 1>
        <rgba 0.8963 0.4287 0.0097 1>
       ]}}
     Random Seed=0
    )
    Lifetime=&MaskFloat(
     Mask Source=1
     Mask Parameter=$Particles.PartitionParticles.Partition
     Mask=0
     Masked Float=&RandomRangeFloat(Minimum=3.5, Maximum=3.9)
     Passthrough Float=&MaskFloat(
      Mask Source=1
      Mask Parameter=$Particles.PartitionParticles.Partition
      Mask=1
      Masked Float=&RandomRangeFloat(Minimum=2.9, Maximum=3.3)
      Passthrough Float=&MaskFloat(
       Mask Source=1
       Mask Parameter=$Particles.PartitionParticles.Partition
       Mask=2
       Masked Float=&RandomRangeFloat(Minimum=2.3, Maximum=2.7)
       Passthrough Float=&MaskFloat(
        Mask Source=1
        Mask Parameter=$Particles.PartitionParticles.Partition
        Mask=3
        Masked Float=&RandomRangeFloat(Minimum=1.7, Maximum=2.1)
        Passthrough Float=&MaskFloat(
         Mask Source=1
         Mask Parameter=$Particles.PartitionParticles.Partition
         Mask=4
         Masked Float=&RandomRangeFloat(Minimum=1.5, Maximum=1.9)
         Passthrough Float=&MaskFloat(
          Mask Source=1
          Mask Parameter=$Particles.PartitionParticles.Partition
          Mask=5
          Masked Float=&RandomRangeFloat(Minimum=1.2, Maximum=1.6)
          Passthrough Float=&MaskFloat(
           Mask Source=1
           Mask Parameter=$Particles.PartitionParticles.Partition
           Mask=6
           Masked Float=&RandomRangeFloat(Minimum=0.8, Maximum=1.2)
           Passthrough Float=&MaskFloat(
            Mask Source=1
            Mask Parameter=$Particles.PartitionParticles.Partition
            Mask=7
            Masked Float=&RandomRangeFloat(Minimum=0.3, Maximum=0.7)
            Passthrough Float=1
           )
          )
         )
        )
       )
      )
     )
    )
    Mesh Scale Mode=::NewEnumerator3
    Mesh Uniform Scale=&Multiply_Float(A=0.08, B=&RandomRangeFloat(Minimum=0.95, Maximum=1))
    Position=$System.DeathPosition
    Position Mode=::NewEnumerator1
    UsePositionOffset=false
    Write Material Random=true
   )
   &ShapeLocation(MaskSpawnGroup=false, Override Seed=false, Sphere Radius=1, UseOwnerScale=true)
   &AddVelocity(
    Module.Use Falloff Distance=false
    Origin Offset Coordinate Space=::Simulation
    Override Seed=false
    Velocity Mode=::NewEnumerator1
    Velocity Speed=&RandomRangeFloat(Minimum=110, Maximum=130)
   )
  ]
  particleUpdateScript [
   &ParticleState()
   &ScaleMeshSize(Scale Factor=&VectorFromFloat(Value=&FloatFromCurve(FloatCurve=DataInterface() {propertyValues {Curve {keys [{interpMode=RCIM_Cubic, tangentMode=RCTM_Break, time=0.7, value=1}, {interpMode=RCIM_Cubic, tangentMode=RCTM_Break, time=0.9, value=0}]}}})))
   &ScaleMeshSize(Scale Factor=&VectorFromFloat(Value=&Lerp_Float(
      A=0
      B=1
      Alpha=&RemapRange(
       Input Value To Remap=&DistanceToCamera(Use Owner Position:NiagaraBool=true, Camera Query:NiagaraDataInterfaceCamera=DataInterface())
       Input Min=50
       Input Max=240
       Output Min=0
       Output Max=1
      )
     ))) {module=ScaleMeshSize002}
   &GravityForce(Gravity=&RandomVector(Vector Scale=&FloatFromCurve(
      FloatCurve=DataInterface() {propertyValues {Curve {keys [
          {interpMode=RCIM_Cubic, tangentMode=RCTM_Break, time=0, value=1}
          {arriveTangent=-2.0604, interpMode=RCIM_Cubic, tangentMode=RCTM_Break, time=0.9, value=0}
         ]}}}
      Scale Curve=&RandomRangeFloat(Minimum=50, Maximum=70)
     )))
   &Drag(Drag=&RandomRangeFloat(Minimum=0.85, Maximum=1.45), Use Linear Drag=true, Use Rotational Drag=false)
   &SolveForcesAndVelocity(Clamp Velocity=false, Limit Acceleration=false)
  ]
  rendererClasses [&NiagaraMeshRendererProperties]
  renderers [{rendererClass=&NiagaraMeshRendererProperties, rendererIndex=0}]
 }
}
summary {
 systemName=NS_Weightless_01
 emitters [
  {emitterName=Main_Ball, rendererClasses [&NiagaraMeshRendererProperties]}
  {emitterName=Main_Orb, rendererClasses []}
  {emitterName=Support_Big_Balls, rendererClasses [&NiagaraMeshRendererProperties]}
  {emitterName=Support_Small_Balls, rendererClasses [&NiagaraMeshRendererProperties]}
  {bEnabled=false, emitterName=Support_Big_Balls001, rendererClasses [&NiagaraMeshRendererProperties]}
  {bEnabled=false, emitterName=Support_Small_Balls001, rendererClasses [&NiagaraMeshRendererProperties]}
 ]
 userVariables []
}
